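# Material function that draws the center dot for Athena HUD reticles as a scalar mask.
# It feeds a Size input into MF_UI_Circle and runs the resulting SDF through two inverted SmoothSteps, a tight one for the dot and a wider one for a soft glow.
# The Dot output gives the tight mask alone, while DotWithGlow adds the dimmed glow to it.
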MaterialFunction /Game/Athena/HUD/reticles/MaterialFunctions/MF_Athena_Reticle_CenterDot
graph {nodes [
  n1 = MaterialFunctionCall(Size=FunctionInput(Preview=Constant() {r=0.035}) {bUsePreviewValueAsDefault=true, inputName=Size, inputType=FunctionInput_Scalar}) {materialFunction=&MF_UI_Circle}
  n2 = SmoothStep(Value=@n1.SDF) {constMax=0.022}
  n3 = SmoothStep(Value=@n1.SDF) {constMax=0.002, constMin=-0.012}
  n4 = Add(A=Multiply(A=OneMinus(None=@n2)) {constB=0.4}, B=FunctionOutput(None=OneMinus(None=@n3)) {outputName=Dot})
  FunctionOutput(None=@n4) {outputName=DotWithGlow}
  NamedRerouteDeclaration(None=@n4) {name=CenterDot, nodeColor=<rgba 0 0 0 1>}
 ]}
